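# Conversation database for the Spark Archer NPC, defining the greeting spoken when a player talks to them and the options that follow.
# A single entry point feeds a Speech node whose contextual lines cover first-ever meetings, specific player outfits and match conversation-count checks.
# Curve-table rows set priority for first-meeting and outfit lines.
# The greeting branches through link nodes to shared NPC library quests, basic services and a Ch5S1.D3.Q8 conversation entry.
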ConversationDatabase /NPCLibrary/NPCs/SparkArcher/NPCLibrary_Conversation_SparkArcher
entry_points [{tag=Conversation.Entry.SparkArcher, destinations [#1]}]
nodes [
 ConversationEntryPointNode_7 {
  id=#1
  class=ConversationEntryPointNode
  outputs [#2]
  properties {entryTag {tagName=Conversation.Entry.SparkArcher}}
 }
 Speech_C_4 {
  id=#2
  class=Speech_C
  outputs [#3, #4, #5]
  properties {
   nodeName=""
   generalConfig {
    defaultMessage="Here to challenge me again? What, you didn't learn your lesson?"
    contextualMessages [
     {
      message=You can bet against me if you want. Just know you'll lose.
      contextRequirements [{requirement {class=FortControllerRequirement_IsFirstEverConversationWithNPC}}]
      priority {curve {rowName=Tandem.Speech.Priority.FirstEverInteraction}}
     }
     {
      message=I'll give you this: it's brave to meet me face-to-face.
      contextRequirements [{requirement {class=FortControllerRequirement_HasCID, properties {allowedCIds [/BRCosmetics/Athena/Items/Cosmetics/Characters/Character_MadameMoth_Posh]}}}]
      priority {curve {rowName=Tandem.Speech.Priority.OutfitReaction}}
     }
     {
      message=Your reputation doesn't scare me. My arrows are faster than your fists.
      contextRequirements [{requirement {
         class=FortControllerRequirement_HasCID
         properties {allowedCIds [&Character_BengalBasher_Suit]}
        }}]
      priority {curve {rowName=Tandem.Speech.Priority.OutfitReaction}}
     }
     {
      message=Fencing? Real fighting doesn't stop when you score a point.
      contextRequirements [{requirement {
         class=FortControllerRequirement_HasCID
         properties {allowedCIds [/BRCosmetics/Athena/Items/Cosmetics/Characters/Character_CeremonialGuard_Fencer]}
        }}]
      priority {curve {rowName=Tandem.Speech.Priority.OutfitReaction}}
     }
     {
      message="Don't be naive, Hope. Valeria won't stop for anyone, even you."
      contextRequirements [{requirement {
         class=FortControllerRequirement_HasCID
         properties {allowedCIds [&Character_VoidRedemption_Rebel]}
        }}]
      priority {curve {rowName=Tandem.Speech.Priority.OutfitReaction}}
     }
     {
      message="Here to challenge me again? What, you didn't learn your lesson?"
      contextRequirements [{requirement {class=FortControllerRequirement_CheckNPCMatchConversationCount, properties {nPCUniqueIdTag {tagName=AISpawnerData.Type.Tandem.SparkArcher}}}}]
      priority {curve {rowName=None}}
     }
     {
      message=Let The Society keep fighting amongst themselves. That's how we'll win.
      contextRequirements [{requirement {class=FortControllerRequirement_CheckNPCMatchConversationCount, properties {nPCUniqueIdTag {tagName=AISpawnerData.Type.Tandem.SparkArcher}}}}]
      priority {curve {rowName=None}}
     }
     {
      message=The Society's losing control. Let's hope they don't make any desperate moves.
      contextRequirements [{requirement {class=FortControllerRequirement_CheckNPCMatchConversationCount, properties {nPCUniqueIdTag {tagName=AISpawnerData.Type.Tandem.SparkArcher}}}}]
      priority {curve {rowName=None}}
     }
     {
      message=Valeria would betray anyone to find those answers she's looking for.
      contextRequirements [{requirement {class=FortControllerRequirement_CheckNPCMatchConversationCount, properties {nPCUniqueIdTag {tagName=AISpawnerData.Type.Tandem.SparkArcher}}}}]
      priority {curve {rowName=None}}
     }
    ]
   }
  }
 }
 ConversationLinkNode_2 {id=#5, class=ConversationLinkNode, properties {remoteEntryTag {tagName=Conversation.Entry.Ch5S1.D3.Q8}}}
 ConversationLinkNode_0 {id=#4, class=ConversationLinkNode, properties {remoteEntryTag {tagName=Conversation.Entry.BasicServices}}}
 ConversationLinkNode_1 {id=#3, class=ConversationLinkNode, properties {remoteEntryTag {tagName=Conversation.Entry.NPCLibrary.Quests}}}
]
